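# Material function for the ColorBreakdown art tool that outputs the final scene color with an overlay applied.
# It samples the PostProcessInput0 scene texture, masks it to RGB and passes it into MF_Overlay as SceneTextureIn.
# The Result of MF_Overlay is the function's only output, so all overlay logic lives in that nested function.
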MaterialFunction /Game/ArtTools/ColorBreakdown/ColorBreakdownModes/MaterialFunctions/MF_FinalColorWithOverlay
graph {nodes [
  n1 = SceneTexture() {sceneTextureId=PPI_PostProcessInput0}
  FunctionOutput(None=MaterialFunctionCall!(SceneTextureIn=ComponentMask(None=@n1.Color) {b=true, g=true, r=true}) {materialFunction=&/MF_Overlay}.Result)
 ]}
