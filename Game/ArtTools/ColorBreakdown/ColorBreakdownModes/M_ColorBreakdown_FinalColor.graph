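# M_ColorBreakdown_FinalColor is a post-process material for the Final Color mode of the ColorBreakdown art tool.
# Its graph holds a single call to the material function MF_FinalColorWithOverlay, named for showing final color with an overlay.
# The material has no logic or parameters of its own, so any color and overlay work lives in that function.
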Material /Game/ArtTools/ColorBreakdown/ColorBreakdownModes/M_ColorBreakdown_FinalColor
graph {
 properties {materialDomain=MD_PostProcess, pixelDepthOffsetMode=PDOM_Legacy, usedShadingModels=Default Lit}
 nodes [MaterialFunctionCall() {materialFunction=/Game/ArtTools/ColorBreakdown/ColorBreakdownModes/MaterialFunctions/MF_FinalColorWithOverlay}]
}
