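# Post-process material for the ColorBreakdown art tool that isolates the luminosity of the image as a single channel.
# MF_FinalColorWithOverlay supplies the color, which MF_RGBtoHSL converts from RGB to HSL.
# A component mask then keeps only the blue channel of that result, the lightness value.
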Material /Game/ArtTools/ColorBreakdown/ColorBreakdownModes/M_ColorBreakdown_Luminosity
graph {
 properties {materialDomain=MD_PostProcess, pixelDepthOffsetMode=PDOM_Legacy, usedShadingModels=Default Lit}
 nodes [ComponentMask(None=MaterialFunctionCall!(RGB=MaterialFunctionCall() {materialFunction=/Game/ArtTools/ColorBreakdown/ColorBreakdownModes/MaterialFunctions/MF_FinalColorWithOverlay}.Result) {materialFunction=/Game/ArtTools/ColorBreakdown/ColorBreakdownModes/MaterialFunctions/MF_RGBtoHSL}.Result) {b=true}]
}
