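# Post-process material in the ColorBreakdown art tools that extracts the saturation of the image as a grayscale value.
# It takes the result of MF_FinalColorWithOverlay, a shared project function, and converts it with the engine RGBtoHSV function.
# A component mask then isolates the green channel, which holds saturation in the 0-1 HSV output.
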Material /Game/ArtTools/ColorBreakdown/ColorBreakdownModes/M_ColorBreakdown_Saturation
graph {
 properties {materialDomain=MD_PostProcess, pixelDepthOffsetMode=PDOM_Legacy, usedShadingModels=Default Lit}
 nodes [ComponentMask(None=MaterialFunctionCall!(RGB=MaterialFunctionCall() {materialFunction=/Game/ArtTools/ColorBreakdown/ColorBreakdownModes/MaterialFunctions/MF_FinalColorWithOverlay}.Result) {materialFunction=/Engine/Functions/Engine_MaterialFunctions02/Math/RGBtoHSV}."RGB to HSV (0-1)") {g=true}]
}
